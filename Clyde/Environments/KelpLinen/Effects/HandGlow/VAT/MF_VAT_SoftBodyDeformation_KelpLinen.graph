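# Houdini VAT soft-body deformation material function for the KelpLinen HandGlow effect, playing baked vertex animation in materials.
# Mesh bounds or legacy parameters encode frame count and texture format, which drive frame UVs from looping time or Display Frame.
# Position textures feed World Position Offset, normals come from rotation quaternions or compressed position alpha, and color textures pass through.
# Static switches toggle interframe interpolation, split positions and an embedded-data debug plane.
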MaterialFunction /Clyde/Environments/KelpLinen/Effects/HandGlow/VAT/MF_VAT_SoftBodyDeformation_KelpLinen
graph {
 properties {
  bExposeToLibrary=true
  description="Supports SideFX Labs Vertex Animation Textures Soft-Body Deformation (Soft) mode. On the Material's Details panel, turn off Tangent Space Normal and set Num Customized UVs to 5. Surface texture UVs are accessible at TexCoord[0]. (All native parameters are inside this Material Function. Create a Material Instance to gain access to the parameters. Avoid creating your own parameters outside this Material Function with names that clash with the parameter names inside this Material Function.)"
  libraryCategoriesText [SideFX]
 }
 nodes [
  n1 = ScalarParameter() {
   desc="To set this parameter, first close the Material Instace. In Content Browser, select the Material Instance and the Data Table before right-clicking on either and then selecting Scripted Actions (or Scripted Asset Actions in UE5) > SideFX Legacy Set VAT Material Instance from Data Table. You need to have exported the Real-Time Data JSON file from Houdini and imported it to Unreal as DataTable (Row Type: SideFX_VAT_Material_Parameters). Alternatively, you can manually copy and paste the value from the JSON file."
   group=Houdini VAT - Instancing
   parameterName=Bound Max X
   sortPriority=1
  }
  n2 = ScalarParameter() {
   desc="To set this parameter, first close the Material Instace. In Content Browser, select the Material Instance and the Data Table before right-clicking on either and then selecting Scripted Actions (or Scripted Asset Actions in UE5) > SideFX Legacy Set VAT Material Instance from Data Table. You need to have exported the Real-Time Data JSON file from Houdini and imported it to Unreal as DataTable (Row Type: SideFX_VAT_Material_Parameters). Alternatively, you can manually copy and paste the value from the JSON file."
   group=Houdini VAT - Instancing
   parameterName=Bound Max Y
   sortPriority=2
  }
  n3 = ScalarParameter() {
   desc="To set this parameter, first close the Material Instace. In Content Browser, select the Material Instance and the Data Table before right-clicking on either and then selecting Scripted Actions (or Scripted Asset Actions in UE5) > SideFX Legacy Set VAT Material Instance from Data Table. You need to have exported the Real-Time Data JSON file from Houdini and imported it to Unreal as DataTable (Row Type: SideFX_VAT_Material_Parameters). Alternatively, you can manually copy and paste the value from the JSON file."
   group=Houdini VAT - Instancing
   parameterName=Bound Max Z
   sortPriority=3
  }
  n4 = ScalarParameter() {
   defaultValue=0.0001
   desc="This value should always be something like: 0.01, 0.001, 0.0001, etc. It controls to what decimal place to round the embedded data. The default value should work in most cases, but if there are visual glitches, try removing one or two zeros to increase this value if your animation spans a large area, or adding one or two zeros to decrease this value if your animation is confined to a small area.\r\n\r\nWithout rounding, the calculations based on the embedded data may be susceptible to floating point errors in Unreal Engine."
   group=Houdini VAT - Debug
   parameterName=Embedded Data Rounding
   sortPriority=4
  }
  n5 = Divide(B=@n4)
  n6 = Multiply(A=Custom() {code="GetPrimitiveData(Parameters.PrimitiveId).LocalObjectBoundsMax.xyz", description=Local Bounds Max, inputs []}, B=@n5) {constB=10000}
  n7 = Multiply(A=@n4) {constB=0.1}
  n8 = StaticSwitchParameter(
   True=Multiply(A=MaterialFunctionCall(X=@n1, Y=@n2, Z=@n3) {materialFunction=&MakeFloat3}.Result) {constB=0.1}
   False=Multiply(A=Round(None=@n6), B=@n7) {constB=0}
  ) {
   desc="If you want to use this Material Instance with ISM/HISM or mesh particles, turn this on and turn on Support Real-Time Instancing in Houdini.\r\n\r\nIf you simply want to use the legacy parameters or modify the object's bounds, turn this on and only turn on Allow Exporting Real-Time Data JSON File (Legacy) in Houdini without turning on Support Real-Time Instancing.\r\n\r\nLegacy parameters are a list of numerical values exported through a JSON file. They contain the embedded data just like the actual mesh, but when Support Legacy Parameters and Instancing is turned on, the shader will read the bounds and the embedded data from the legacy parameters instead of the actual mesh.\r\n\r\nUsing legacy parameters is less convenient, but it does produce more accurate results if the animation spans a huge area; it also leads to lower instruction counts."
   group=Houdini VAT - Instancing
   parameterName=Support Legacy Parameters and Instancing
   sortPriority=0
  }
  n9 = ComponentMask(None=@n8) {g=true}
  n10 = Subtract(A=@n9, B=Floor(None=@n9)) {desc="0.7/0.8 if texture data are in HDR.\r\n0.3/0.4 if texture data are in LDR."}
  n11 = ComponentMask(None=@n8) {b=true}
  n12 = Floor(None=@n11)
  n13 = Divide(B=OneMinus(None=Subtract(A=@n11, B=@n12))) {desc=Max Custom Data}
  n14 = MaterialFunctionCall(X=@n10, Y=@n13) {materialFunction=&MakeFloat3}
  n15 = Constant() {r=8}
  n16 = ScalarParameter() {
   desc="To set this parameter, first close the Material Instace. In Content Browser, select the Material Instance and the Data Table before right-clicking on either and then selecting Scripted Actions (or Scripted Asset Actions in UE5) > SideFX Legacy Set VAT Material Instance from Data Table. You need to have exported the Real-Time Data JSON file from Houdini and imported it to Unreal as DataTable (Row Type: SideFX_VAT_Material_Parameters). Alternatively, you can manually copy and paste the value from the JSON file."
   group=Houdini VAT - Instancing
   parameterName=Bound Min X
   sortPriority=4
  }
  n17 = ScalarParameter() {
   desc="To set this parameter, first close the Material Instace. In Content Browser, select the Material Instance and the Data Table before right-clicking on either and then selecting Scripted Actions (or Scripted Asset Actions in UE5) > SideFX Legacy Set VAT Material Instance from Data Table. You need to have exported the Real-Time Data JSON file from Houdini and imported it to Unreal as DataTable (Row Type: SideFX_VAT_Material_Parameters). Alternatively, you can manually copy and paste the value from the JSON file."
   group=Houdini VAT - Instancing
   parameterName=Bound Min Y
   sortPriority=5
  }
  n18 = ScalarParameter() {
   desc="To set this parameter, first close the Material Instace. In Content Browser, select the Material Instance and the Data Table before right-clicking on either and then selecting Scripted Actions (or Scripted Asset Actions in UE5) > SideFX Legacy Set VAT Material Instance from Data Table. You need to have exported the Real-Time Data JSON file from Houdini and imported it to Unreal as DataTable (Row Type: SideFX_VAT_Material_Parameters). Alternatively, you can manually copy and paste the value from the JSON file."
   group=Houdini VAT - Instancing
   parameterName=Bound Min Z
   sortPriority=6
  }
  n19 = Multiply(A=Custom() {code="GetPrimitiveData(Parameters.PrimitiveId).LocalObjectBoundsMin.xyz", description=Local Bounds Min, inputs []}, B=@n5) {constB=10000}
  n20 = StaticSwitchParameter(
   True=Multiply(A=MaterialFunctionCall(X=@n16, Y=@n17, Z=@n18) {materialFunction=&MakeFloat3}.Result) {constB=0.1}
   False=Multiply(A=Round(None=@n19), B=@n7) {constB=0}
  ) {
   desc="If you want to use this Material Instance with ISM/HISM or mesh particles, turn this on and turn on Support Real-Time Instancing in Houdini.\r\n\r\nIf you simply want to use the legacy parameters or modify the object's bounds, turn this on and only turn on Allow Exporting Real-Time Data JSON File (Legacy) in Houdini without turning on Support Real-Time Instancing.\r\n\r\nLegacy parameters are a list of numerical values exported through a JSON file. They contain the embedded data just like the actual mesh, but when Support Legacy Parameters and Instancing is turned on, the shader will read the bounds and the embedded data from the legacy parameters instead of the actual mesh.\r\n\r\nUsing legacy parameters is less convenient, but it does produce more accurate results if the animation spans a huge area; it also leads to lower instruction counts."
   group=Houdini VAT - Instancing
   parameterName=Support Legacy Parameters and Instancing
   sortPriority=0
  }
  n21 = ComponentMask(None=@n20) {g=true}
  n22 = OneMinus(None=Subtract(A=Ceil(None=@n21), B=@n21)) {desc=Active Pixels Ratio X}
  n23 = ComponentMask(None=@n20) {r=true}
  n24 = OneMinus(None=Subtract(A=Ceil(None=@n23), B=@n23)) {desc=Active Pixels Ratio Y}
  n25 = TextureObjectParameter() {
   desc=The RGB channels store positions. The A channel can also store compressed normals or custom float data.
   group=Houdini VAT
   parameterName=Position Texture
   samplerType=SAMPLERTYPE_LinearColor
   sortPriority=10
   texture=/Engine/Functions/Engine_MaterialFunctions02/PivotPainter2/Black_1x1_EXR_Texture
  }
  n26 = ComponentMask(None=TextureProperty(Tex=@n25)) {g=true}
  n27 = ComponentMask(None=@n20) {b=true}
  n28 = Subtract(A=Ceil(None=@n27), B=@n27)
  n29 = Divide(A=Round(None=Multiply(A=@n26, B=@n24)), B=Round(None=Multiply(A=@n28) {constB=100})) {desc=Animation frame count}
  n30 = MaterialFunctionCall(X=@n22, Y=@n24, Z=@n29) {materialFunction=&MakeFloat3}
  n31 = Add(A=TextureCoordinate(), B=Constant2Vector() {g=-0.05})
  n32 = TextureCoordinate() {coordinateIndex=1}
  n33 = Multiply(A=@n22, B=ComponentMask(None=@n32) {r=true}) {constA=1}
  n34 = Multiply(A=ComponentMask(None=@n32) {g=true}, B=@n24)
  n35 = FunctionInput(Preview=Time()) {
   bUsePreviewValueAsDefault=true
   description="The animation will use this input when the Material Instance parameter Auto Play is turned on. Otherwise, the animation uses the parameter Display Frame."
   inputName="Time (VS)"
   inputType=FunctionInput_Scalar
   sortPriority=-2
  }
  n36 = ScalarParameter() {
   desc="Used when Auto Play is turned on, this helps you start the animation from its first frame when a specific gameplay event is triggered.\r\n\r\nUse Blueprints or C++ code to set this parameter through a Material Instance Dynamic. The value should be set to the output of the Blueprint node Get Time Seconds (or the C++ function UWorld::GetTimeSeconds) the moment this event is triggered. It should only be set once.\r\n\r\nThis method is compatible with Motion Blur, whereas setting Display Frame is not. To fully enable Motion Blur, turn on Output Velocities during Base Pass in Project Settings."
   group=Houdini VAT
   parameterName=Game Time at First Frame
   sortPriority=-9
  }
  n37 = ScalarParameter() {
   defaultValue=60
   desc="The FPS of the Houdini HIP file when exporting the animation. It is possible for you to simulate at a higher FPS, cache the data, and then export at a lower FPS. This parameter refers to the export FPS. Make sure this parameter is set correctly whenever it is exposed on the Material Instance interface."
   group=Houdini VAT
   parameterName=Houdini FPS
   sortPriority=-6
  }
  n38 = Multiply(A=Subtract(A=@n35, B=@n36), B=Divide(A=@n37, B=Subtract(A=@n29) {constB=0.01}))
  n39 = ScalarParameter() {
   defaultValue=1
   desc="Use this to scale playback speed. When this is 1.0 and Original FPS is correct, the animation will be playing at its original speed."
   group=Houdini VAT
   parameterName=Playback Speed
   sortPriority=-7
  }
  n40 = Multiply(A=Frac(None=Multiply(A=@n38, B=@n39)), B=@n29)
  n41 = ScalarParameter() {
   defaultValue=1
   desc="The exact animation frame to display, unaffected by Playback Speed or Houdini FPS. When this is set to 1.0, it displays the first frame of the animation. This frame sequence exactly matches the frame sequence in Houdini. If the Interframe Interpolation parameter is available and turned on, fractional values in Display Frame will generate interpolated results between adjacent frames.\r\n\r\nUse Blueprints or C++ code to drive this parameter through a Material Instance Dynamic.\r\n\r\nThis method is incompatible with Motion Blur. To support Motion Blur, turn on Auto Play and follow the instructions in the tooltip of Game Time at First Frame."
   group=Houdini VAT
   parameterName=Display Frame
   sliderMax=1000
   sliderMin=1
   sortPriority=-8
  }
  n42 = FunctionInput(Preview=@n41) {
   bUsePreviewValueAsDefault=true
   description=Overrides the Material Instance parameter Display Frame. The animation will ignore this input when the Material Instance parameter Auto Play is turned on.
   inputName="Display Frame Override (VS)"
   inputType=FunctionInput_Scalar
   sortPriority=-1
  }
  n43 = StaticSwitchParameter(True=Add(A=Floor(None=@n40)), False=Floor(None=@n42)) {
   defaultValue=true
   desc="Automatically plays back the animation in loops. The animation speed is affected by both Playback Speed and Houdini FPS.\r\n\r\nFor this mode, turn this on if you want to support Motion Blur."
   group=Houdini VAT
   parameterName=Auto Playback
   sortPriority=-10
  }
  n44 = Fmod(A=Subtract(A=@n43), B=@n29)
  n45 = Divide(B=@n29)
  n46 = Add(A=@n34, B=Multiply(A=Multiply(A=@n44, B=@n45), B=@n24))
  n47 = AppendVector(A=@n33, B=@n46) {desc=This frame UV}
  n48 = TextureObjectParameter() {
   desc="The data can come from the \"Cd\" and \"Alpha\" attributes, or custom vector and float attributes."
   group=Houdini VAT
   parameterName=Color Texture
   samplerType=SAMPLERTYPE_LinearColor
   sortPriority=13
   texture=&Black_1x1_EXR_Texture
  }
  n49 = TextureSample(UVs=@n47, Tex=@n48) {texture=&Black}
  n50 = Fmod(A=@n43, B=@n29)
  n51 = Add(A=@n34, B=Multiply(A=Multiply(A=@n50, B=@n45), B=@n24))
  n52 = AppendVector(A=@n33, B=@n51) {desc=Next frame UV}
  n53 = TextureSample(UVs=@n52, Tex=@n48) {texture=&Black}
  n54 = StaticSwitchParameter(True=@n40, False=@n42) {
   defaultValue=true
   desc="Automatically plays back the animation in loops. The animation speed is affected by both Playback Speed and Houdini FPS.\r\n\r\nFor this mode, turn this on if you want to support Motion Blur."
   group=Houdini VAT
   parameterName=Auto Playback
   sortPriority=-10
  }
  n55 = Frac(None=@n54) {desc=Interpolation alpha}
  n56 = StaticSwitchParameter(True=LinearInterpolate(A=@n49.RGBA, B=@n53.RGBA, Alpha=@n55), False=@n49.RGBA) {
   desc="There is a cost associated with interpolating each vertex animation texture. While Interframe Interpolation is the overall switch, you can independently decide whether or not to interpolate Color Texture. This only affects the pixel shader outputs Color RGB (PS) and Color Alpha (PS). The outputs ending with This Frame (VS) / Next Frame (VS) are never interpolated."
   group=Houdini VAT
   parameterName=Interpolate Color
   sortPriority=2
  }
  n57 = StaticSwitchParameter(True=@n56, False=@n49.RGBA) {
   desc=Interpolates interframe data when the animation frame is a fractional number. This results in smooth visuals even when you slow down the animation or when the frame rate is unstable.
   group=Houdini VAT
   parameterName=Interframe Interpolation
   sortPriority=1
  }
  n58 = VertexInterpolator(VS=@n57)
  n59 = StaticSwitchParameter(
   True=Add(
    A=MaterialFunctionCall!(Vector3=@n14.Result, MaximumNumberOfDigits=@n15, UVs, DebugTextLocation RG_UpperRight BA_LowerLeft, Component Spacing) {materialFunction=/Engine/Functions/Engine_MaterialFunctions02/Utility/DebugFloat3Values}.GreyScaleOutput
    B=MaterialFunctionCall(Vector3=@n30.Result, MaximumNumberOfDigits=@n15, UVs=@n31) {materialFunction=&DebugFloat3Values}.ColorCodedOutput
   )
   False=ComponentMask(None=@n58.PS) {b=true, g=true, r=true}
  ) {
   desc="To use this feature, Include Hidden Debug Plane in Geometry must be turned on in Houdini and this Material Function's Color RGB (PS) output must be connected to Base Color (at least during debugging). In the debug mode, the data embedded in the static mesh are displayed on a plane (facing local-space +Y).\r\n\r\nThe greyscale XYZ values are: \r\n\r\ntexture format and the number of position textures (0.7 or 0.8 = HDR; 0.3 or 0.4 = non-HDR; 0.3 or 0.7 = one texture; 0.4 or 0.8 = two textures);\r\n\r\nmax value of a custom attribute in Position Texture Alpha (no data if no custom attribute is present; compressed normals are not considered a custom attribute);\r\n\r\nno data.\r\n\r\nThe RGB values are: \r\n\r\npowers of two mode active pixels ratio X;\r\n\r\npowers of two mode active pixels ratio Y;\r\n\r\nanimation frame count.\r\n\r\n(Due to some numerical errors in the Unreal debug nodes, occasionally when an embedded value is actually 1 or 0, the plane may misleadingly display 0 or 1 . The does not affect the actual animation and you can safely ignore it. You can verify this is indeed a display bug by temporarily add or remove a few zeros in Embedded Data Rounding.)\r\n\r\nWhen not in the debug mode, the debug plane is invisible and incurs next to no cost."
   group=Houdini VAT - Debug
   parameterName=Debug Embedded Data
   sortPriority=3
  }
  FunctionOutput(None=@n59) {
   description="Outputs Color Texture RGB on the pixel shader (PS). The data can come from the \"Cd\" attribute or a custom vector attribute. The data can be automatically interpolated between frames based on the Material Instance settings. To process this frame's data on the vertex shader, use Color RGBA This Frame (VS). To get interpolated data on the vertex shader, use Color RGBA This Frame (VS), Color RGBA Next Frame (VS), and Interframe Interpolation Alpha (VS) to create your own Lerp."
   outputName="Color RGB (PS)"
  }
  n60 = ComponentMask(None=TextureCoordinate() {coordinateIndex=1}) {desc="Boundary data triangles and debug\r\nplane have a constant V of 1.0,\r\nso they can be easily masked out.", g=true}
  n61 = Constant() {r=0.9}
  n62 = PreSkinnedPosition()
  n63 = Subtract(A=Constant3Vector(), B=@n62) {desc=Invisible vertices LPO}
  n64 = TextureCoordinate()
  n65 = ComponentMask(None=@n64) {r=true}
  n66 = ComponentMask(None=@n64) {g=true}
  n67 = Subtract(
   A=MaterialFunctionCall(X=ConstantBiasScale(None=@n65) {bias=-0.5, scale=500}, Z=ConstantBiasScale(None=@n66) {bias=-1, scale=-500}) {materialFunction=&MakeFloat3}.Result
   B=@n62
  ) {desc=Debug plane LPO}
  n68 = TextureSample(UVs=@n47, Tex=@n25) {texture=&Black}
  n69 = TextureObjectParameter() {
   desc=The RGB channels store positions. See the tooltip of Positions Require Two Textures.
   group=Houdini VAT
   parameterName=Position Texture 2
   samplerType=SAMPLERTYPE_LinearColor
   sortPriority=11
   texture=&Black_1x1_EXR_Texture
  }
  n70 = TextureSample(UVs=@n47, Tex=@n69) {texture=&Black}
  n71 = StaticSwitchParameter(True=Add(A=@n68.RGB, B=Multiply(A=@n70.RGB) {constB=0.01}), False=@n68.RGB) {
   desc="This setting's value must match that of Split Positions into Two Textures in Houdini. When the Houdini counterpart is turned on, the position data are split into two parts and encoded in the two position textures, so you have to use them both."
   group=Houdini VAT
   parameterName=Positions Require Two Textures
   sortPriority=7
  }
  n72 = Multiply(A=@n20) {constB=10}
  n73 = Subtract(A=Multiply(A=@n8) {constB=10}, B=@n72)
  n74 = If(A=@n10, B=Constant() {r=0.5}, "A > B"=@n71, "A == B"=@n71, "A < B"=Add(A=Multiply(A=@n71, B=@n73) {constB=20}, B=@n72)) {constB=0.5, desc=This frame Position}
  n75 = TextureSample(UVs=@n52, Tex=@n25) {texture=&Black}
  n76 = TextureSample(UVs=@n52, Tex=@n69) {texture=&Black}
  n77 = StaticSwitchParameter(True=Add(A=@n75.RGB, B=Multiply(A=@n76.RGB) {constB=0.01}), False=@n75.RGB) {
   desc="This setting's value must match that of Split Positions into Two Textures in Houdini. When the Houdini counterpart is turned on, the position data are split into two parts and encoded in the two position textures, so you have to use them both."
   group=Houdini VAT
   parameterName=Positions Require Two Textures
   sortPriority=7
  }
  n78 = If(A=@n10, B=Constant() {r=0.5}, "A > B"=@n77, "A == B"=@n77, "A < B"=Add(A=Multiply(A=@n77, B=@n73) {constB=20}, B=@n72)) {constB=0.5, desc=Next frame Position}
  n79 = StaticSwitchParameter(True=LinearInterpolate(A=@n74, B=@n78, Alpha=@n55), False=@n74) {
   desc=Interpolates interframe data when the animation frame is a fractional number. This results in smooth visuals even when you slow down the animation or when the frame rate is unstable.
   group=Houdini VAT
   parameterName=Interframe Interpolation
   sortPriority=1
  }
  n80 = FunctionInput() {
   bUsePreviewValueAsDefault=true
   description="Provides an additional local-space offset to the animated vertex position. This input can vary from point to point (a point is where multiple vertices from adjacent polygons visually overlap with one another). To randomize the offset per point, you can use MF_SimpleRandom or define your own random function. Generate a seed for that function using TexCoord[1]'s RG components, as each point has a unique pair of UV values in TexCoord[1]."
   inputName="Additional Local Position Offset (VS)"
   sortPriority=1
  }
  n81 = StaticSwitchParameter(
   True=If(A=@n60, B=@n61, "A > B"=@n67, "A == B"=@n67, "A < B"=@n63)
   False=If(A=@n60, B=@n61, "A > B"=@n63, "A == B"=@n63, "A < B"=Add(A=@n79, B=@n80) {desc=Regular LPO})
  ) {
   desc="To use this feature, Include Hidden Debug Plane in Geometry must be turned on in Houdini and this Material Function's Color RGB (PS) output must be connected to Base Color (at least during debugging). In the debug mode, the data embedded in the static mesh are displayed on a plane (facing local-space +Y).\r\n\r\nThe greyscale XYZ values are: \r\n\r\ntexture format and the number of position textures (0.7 or 0.8 = HDR; 0.3 or 0.4 = non-HDR; 0.3 or 0.7 = one texture; 0.4 or 0.8 = two textures);\r\n\r\nmax value of a custom attribute in Position Texture Alpha (no data if no custom attribute is present; compressed normals are not considered a custom attribute);\r\n\r\nno data.\r\n\r\nThe RGB values are: \r\n\r\npowers of two mode active pixels ratio X;\r\n\r\npowers of two mode active pixels ratio Y;\r\n\r\nanimation frame count.\r\n\r\n(Due to some numerical errors in the Unreal debug nodes, occasionally when an embedded value is actually 1 or 0, the plane may misleadingly display 0 or 1 . The does not affect the actual animation and you can safely ignore it. You can verify this is indeed a display bug by temporarily add or remove a few zeros in Embedded Data Rounding.)\r\n\r\nWhen not in the debug mode, the debug plane is invisible and incurs next to no cost."
   group=Houdini VAT - Debug
   parameterName=Debug Embedded Data
   sortPriority=3
  }
  FunctionOutput(None=Transform(None=@n81) {transformSourceType=TRANSFORMSOURCE_Local}) {description=Connect this to World Position Offset., outputName=World Position Offset, sortPriority=4}
  n82 = Constant3Vector()
  n83 = TextureObjectParameter() {
   desc="The data strictly come from the \"Cd2\" and \"Alpha2\" attributes. You can use Spare Color Texture to store the actual colors and Color Texture to store custom attributes."
   group=Houdini VAT
   parameterName=Spare Color Texture
   samplerType=SAMPLERTYPE_LinearColor
   sortPriority=14
   texture=&Black_1x1_EXR_Texture
  }
  n84 = TextureSample(UVs=@n47, Tex=@n83) {texture=&Black}
  n85 = TextureSample(UVs=@n52, Tex=@n83) {texture=&Black}
  n86 = StaticSwitchParameter(True=LinearInterpolate(A=@n84.RGBA, B=@n85.RGBA, Alpha=@n55), False=@n84.RGBA) {
   desc="There is a cost associated with interpolating each vertex animation texture. While Interframe Interpolation is the overall switch, you can independently decide whether or not to interpolate Spare Color Texture. This only affects the pixel shader output Spare Color RGBA (PS). The outputs ending with This Frame (VS) / Next Frame (VS) are never interpolated."
   group=Houdini VAT
   parameterName=Interpolate Spare Color
   sortPriority=3
  }
  n87 = StaticSwitchParameter(True=@n86, False=@n84.RGBA) {
   desc=Interpolates interframe data when the animation frame is a fractional number. This results in smooth visuals even when you slow down the animation or when the frame rate is unstable.
   group=Houdini VAT
   parameterName=Interframe Interpolation
   sortPriority=1
  }
  n88 = StaticSwitchParameter(True=@n82, False=VertexInterpolator(VS=@n87).PS) {
   desc="To use this feature, Include Hidden Debug Plane in Geometry must be turned on in Houdini and this Material Function's Color RGB (PS) output must be connected to Base Color (at least during debugging). In the debug mode, the data embedded in the static mesh are displayed on a plane (facing local-space +Y).\r\n\r\nThe greyscale XYZ values are: \r\n\r\ntexture format and the number of position textures (0.7 or 0.8 = HDR; 0.3 or 0.4 = non-HDR; 0.3 or 0.7 = one texture; 0.4 or 0.8 = two textures);\r\n\r\nmax value of a custom attribute in Position Texture Alpha (no data if no custom attribute is present; compressed normals are not considered a custom attribute);\r\n\r\nno data.\r\n\r\nThe RGB values are: \r\n\r\npowers of two mode active pixels ratio X;\r\n\r\npowers of two mode active pixels ratio Y;\r\n\r\nanimation frame count.\r\n\r\n(Due to some numerical errors in the Unreal debug nodes, occasionally when an embedded value is actually 1 or 0, the plane may misleadingly display 0 or 1 . The does not affect the actual animation and you can safely ignore it. You can verify this is indeed a display bug by temporarily add or remove a few zeros in Embedded Data Rounding.)\r\n\r\nWhen not in the debug mode, the debug plane is invisible and incurs next to no cost."
   group=Houdini VAT - Debug
   parameterName=Debug Embedded Data
   sortPriority=3
  }
  FunctionOutput(None=@n88) {
   description="Outputs Spare Color Texture RGBA on the pixel shader (PS). The data strictly come from the \"Cd2\" and \"Alpha2\" attributes. You can use Spare Color Texture to store the actual colors and Color Texture to store custom attributes. The data can be automatically interpolated between frames based on the Material Instance settings. To process this frame's data on the vertex shader, use Spare Color RGBA This Frame (VS). To get interpolated data on the vertex shader, use Spare Color RGBA This Frame (VS), Spare Color RGBA Next Frame (VS), and Interframe Interpolation Alpha (VS) to create your own Lerp."
   outputName="Spare Color RGBA (PS)"
   sortPriority=2
  }
  FunctionOutput(None=@n53.RGBA) {
   description="Outputs next frame's Color Texture RGBA on the vertex shader (VS). The data can come from the \"Cd\" and \"Alpha\" attributes, or custom vector and float attributes. To get interpolated data, use this with Color RGBA This Frame (VS) and Interframe Interpolation Alpha (VS) to create your own Lerp."
   outputName="Color RGBA Next Frame (VS)"
   sortPriority=12
  }
  n89 = Multiply(A=@n68.A) {constB=1024}
  n90 = Floor(None=Divide(A=@n89) {constB=32})
  n91 = Divide(A=@n90) {constB=31.5}
  n92 = Divide(A=Subtract(A=@n89, B=Multiply(A=@n90) {constB=32})) {constB=31.5}
  n93 = Subtract(A=Multiply(A=AppendVector(A=@n91, B=@n92)) {constB=4}) {constB=2}
  n94 = DotProduct(A=@n93, B=@n93)
  n95 = SquareRoot(None=OneMinus(None=Divide(A=@n94) {constB=4}))
  n96 = Divide(A=@n94)
  n97 = Clamp(None=AppendVector(A=Multiply(A=@n95, B=@n93), B=OneMinus(None=@n96))) {minDefault=-1}
  n98 = Multiply(A=@n75.A) {constB=1024}
  n99 = Floor(None=Divide(A=@n98) {constB=32})
  n100 = Divide(A=@n99) {constB=31.5}
  n101 = Divide(A=Subtract(A=@n98, B=Multiply(A=@n99) {constB=32})) {constB=31.5}
  n102 = Subtract(A=Multiply(A=AppendVector(A=@n100, B=@n101)) {constB=4}) {constB=2}
  n103 = DotProduct(A=@n102, B=@n102)
  n104 = SquareRoot(None=OneMinus(None=Divide(A=@n103) {constB=4}))
  n105 = Divide(A=@n103)
  n106 = Clamp(None=AppendVector(A=Multiply(A=@n104, B=@n102), B=OneMinus(None=@n105))) {minDefault=-1}
  n107 = StaticSwitchParameter(True=LinearInterpolate(A=@n97, B=@n106, Alpha=@n55), False=@n97) {
   desc=Interpolates interframe data when the animation frame is a fractional number. This results in smooth visuals even when you slow down the animation or when the frame rate is unstable.
   group=Houdini VAT
   parameterName=Interframe Interpolation
   sortPriority=1
  }
  n108 = Transform(None=@n107) {transformSourceType=TRANSFORMSOURCE_Local}
  n109 = FunctionInput() {
   bUsePreviewValueAsDefault=true
   description="The RGB output of a regular surface normal map or the combined RGB output of multiple blended surface normal maps. Do not input any world-space normal here. Instead, blend any world-space normal with this Material Function's Normal output."
   inputName="Surface Normal Maps RGB (PS)"
   previewValue=<0 0 1 1>
  }
  n110 = AppendVector(A=ComponentMask(None=TextureCoordinate() {coordinateIndex=3}) {g=true}, B=TextureCoordinate() {coordinateIndex=4}) {desc=Tangent}
  n111 = AppendVector(A=TextureCoordinate() {coordinateIndex=2}, B=ComponentMask(None=TextureCoordinate() {coordinateIndex=3}) {r=true}) {desc=Normal}
  n112 = StaticSwitchParameter(
   True=MaterialFunctionCall(VectorToTransform=@n109, BasisX=@n110, BasisY=CrossProduct(A=@n111, B=@n110), BasisZ=@n111) {materialFunction=&Transform3x3Matrix}.Transformed Vector
   False=@n111
  ) {
   defaultValue=true
   desc="Turn this on if you need to apply surface normal maps. When importing the FBX, set Normal Import Method to \"Import Normals and Tangents\"."
   group=Houdini VAT
   parameterName=Support Surface Normal Maps
   sortPriority=5
  }
  n113 = StaticSwitchParameter(True=VertexInterpolator(VS=Normalize(VectorInput=@n108)).PS, False=@n112) {
   desc="Uses the medium-quality compressed normals from Position Texture Alpha. Only turn this on if Export Compressed Normals in Position Alpha is turned on in Houdini. This option removes the need for Rotation Texture, but the drawback is that you cannot apply any surface normal maps. If you do not need surface normal maps but still prefer higher-quality normals, turn off this and turn off Support Surface Normal Maps to retrieve normals from Rotation Texture."
   group=Houdini VAT
   parameterName=Use Compressed Normals
   sortPriority=4
  }
  n114 = StaticSwitchParameter(True=@n113, False=@n112) {
   desc="The data in Position Texture Alpha can come from the compressed normals or a custom float attribute. Turn this off if they come from the latter or if there is no data. It will disallow the use of compressed normals and interprets the data in this channel as a float, as opposed to a compressed 3D vector. This Material Function's Data in Position Alpha This Frame (VS) and Data in Position Alpha Next Frame (VS) outputs will always give you the appropriate interpretation of the data based on this parameter."
   group=Houdini VAT - Customization
   parameterName=Compressed Normals Are in Position Alpha
   sortPriority=0
  }
  n115 = StaticSwitchParameter(True=Multiply(A=@n114, B=TwoSidedSign()), False=@n114) {
   desc=Turn this on if you want both sides of polygons to be rendered. You also need to turn on Two Sided on the Material's Details panel.
   group=Houdini VAT
   parameterName=Two Sided Normals
   sortPriority=6
  }
  n116 = StaticSwitchParameter(True=@n82, False=@n115) {
   desc="To use this feature, Include Hidden Debug Plane in Geometry must be turned on in Houdini and this Material Function's Color RGB (PS) output must be connected to Base Color (at least during debugging). In the debug mode, the data embedded in the static mesh are displayed on a plane (facing local-space +Y).\r\n\r\nThe greyscale XYZ values are: \r\n\r\ntexture format and the number of position textures (0.7 or 0.8 = HDR; 0.3 or 0.4 = non-HDR; 0.3 or 0.7 = one texture; 0.4 or 0.8 = two textures);\r\n\r\nmax value of a custom attribute in Position Texture Alpha (no data if no custom attribute is present; compressed normals are not considered a custom attribute);\r\n\r\nno data.\r\n\r\nThe RGB values are: \r\n\r\npowers of two mode active pixels ratio X;\r\n\r\npowers of two mode active pixels ratio Y;\r\n\r\nanimation frame count.\r\n\r\n(Due to some numerical errors in the Unreal debug nodes, occasionally when an embedded value is actually 1 or 0, the plane may misleadingly display 0 or 1 . The does not affect the actual animation and you can safely ignore it. You can verify this is indeed a display bug by temporarily add or remove a few zeros in Embedded Data Rounding.)\r\n\r\nWhen not in the debug mode, the debug plane is invisible and incurs next to no cost."
   group=Houdini VAT - Debug
   parameterName=Debug Embedded Data
   sortPriority=3
  }
  FunctionOutput(None=@n116) {
   description="Outputs the world-space normal. On Material Details panel, turn off Tangent Space Normal. Connect this to Normal."
   outputName="Normal (Tangent Space Normal Off)"
   sortPriority=3
  }
  FunctionOutput(None=@n85.RGBA) {
   description="Outputs next frame's Spare Color Texture RGBA on the vertex shader (VS). The data strictly come from the \"Cd2\" and \"Alpha2\" attributes. You can use Spare Color Texture to store the actual colors and Color Texture to store custom attributes. To get interpolated data, use this with Spare Color RGBA This Frame (VS) and Interframe Interpolation Alpha (VS) to create your own Lerp."
   outputName="Spare Color RGBA Next Frame (VS)"
   sortPriority=14
  }
  n117 = StaticSwitchParameter(
   True=Normalize(VectorInput=Transform(None=@n106) {transformSourceType=TRANSFORMSOURCE_Local})
   False=Multiply(A=@n75.A, B=@n13) {desc=Original custom data next frame}
  ) {
   desc="The data in Position Texture Alpha can come from the compressed normals or a custom float attribute. Turn this off if they come from the latter or if there is no data. It will disallow the use of compressed normals and interprets the data in this channel as a float, as opposed to a compressed 3D vector. This Material Function's Data in Position Alpha This Frame (VS) and Data in Position Alpha Next Frame (VS) outputs will always give you the appropriate interpretation of the data based on this parameter."
   group=Houdini VAT - Customization
   parameterName=Compressed Normals Are in Position Alpha
   sortPriority=0
  }
  FunctionOutput(None=@n117) {
   description="Unlike other texture channels, Position Texture Alpha tends to store data in a special way. The data usually go through some form of normalization or compression. This output gives you next frame's data in their originally intended format, not the raw data in Position Texture Alpha. To get interpolated data, use this with Data in Position Alpha This Frame (VS) and Interframe Interpolation Alpha (VS) to create your own Lerp."
   outputName="Data in Position Alpha Next Frame (VS)"
   sortPriority=10
  }
  FunctionOutput(None=@n46) {
   description="On the vertex shader (VS), outputs the progress of the animation represented by this frame's V (as in UVs) used to sample the vertex animation textures. To get interpolated values, use this with Animation Phase Next Frame (VS) and Interframe Interpolation Alpha (VS) to create your own Lerp."
   outputName="Animation Phase This Frame (0-1) (VS)"
   sortPriority=5
  }
  FunctionOutput(None=@n49.RGBA) {
   description="Outputs this frame's Color Texture RGBA on the vertex shader (VS). The data can come from the \"Cd\" and \"Alpha\" attributes, or custom vector and float attributes. To get interpolated data, use this with Color RGBA Next Frame (VS) and Interframe Interpolation Alpha (VS) to create your own Lerp."
   outputName="Color RGBA This Frame (VS)"
   sortPriority=11
  }
  FunctionOutput(None=@n84.RGBA) {
   description="Outputs this frame's Spare Color Texture RGBA on the vertex shader (VS). The data strictly come from the \"Cd2\" and \"Alpha2\" attributes. You can use Spare Color Texture to store the actual colors and Color Texture to store custom attributes. To get interpolated data, use this with Spare Color RGBA Next Frame (VS) and Interframe Interpolation Alpha (VS) to create your own Lerp."
   outputName="Spare Color RGBA This Frame (VS)"
   sortPriority=13
  }
  FunctionOutput(None=TextureCoordinate() {coordinateIndex=1}) {
   description="You have two options: either connect this to Customized UV1 or remember not to connect anything to Customized UV1. This output's only purpose is to prevent accidental overwritting of UV1 in the Material, which must not be allowed."
   outputName="Customized UV1 (*)"
   sortPriority=21
  }
  FunctionOutput(None=@n51) {
   description="On the vertex shader (VS), outputs the progress of the animation represented by next frame's V (as in UVs) used to sample the vertex animation textures. To get interpolated values, use this with Animation Phase This Frame (VS) and Interframe Interpolation Alpha (VS) to create your own Lerp."
   outputName="Animation Phase Next Frame (0-1) (VS)"
   sortPriority=6
  }
  n118 = PreSkinnedPosition() {desc=Compatible with SM/ISM/HISM}
  FunctionOutput(None=Add(A=@n74, B=@n118)) {
   description="Outputs this frame's point local position on the vertex shader (VS). If Position Texture is non-HDR, this output is Position Texture RGB transformed by the object local bounds. To get interpolated data, use this with the outputs Point Local Position Next Frame (VS) and Interframe Interpolation Alpha (VS) to create your own Lerp."
   outputName="Point Local Position This Frame (VS)"
   sortPriority=7
  }
  FunctionOutput(None=Add(A=@n78, B=@n118)) {
   description="Outputs next frame's point local position on the vertex shader (VS). If Position Texture is non-HDR, this output is Position Texture RGB transformed by the object local bounds. To get interpolated data, use this with the outputs Point Local Position This Frame (VS) and Interframe Interpolation Alpha (VS) to create your own Lerp."
   outputName="Point Local Position Next Frame (VS)"
   sortPriority=8
  }
  n119 = Constant2Vector()
  n120 = TextureObjectParameter() {
   desc=Describes point rotations. The RGBA channels store quaternions that simultaneously describe normal and tangent rotations. This replaces the legacy Normal Texture.
   group=Houdini VAT
   parameterName=Rotation Texture
   samplerType=SAMPLERTYPE_LinearColor
   sortPriority=12
   texture=&Black_1x1_EXR_Texture
  }
  n121 = TextureSample(UVs=@n52, Tex=@n120) {texture=&Black}
  n122 = If(A=@n10, B=Constant() {r=0.5}, "A > B"=@n121.RGBA, "A == B"=@n121.RGBA, "A < B"=ConstantBiasScale(None=@n121.RGBA) {bias=-0.5, scale=2}) {constB=0.5}
  n123 = ComponentMask(None=@n122) {b=true, desc=Quaternion XYZ, g=true, r=true}
  n124 = ComponentMask(None=@n122) {a=true, desc=Quaternion W}
  n125 = Constant3Vector() {constant=<rgba 0 0 1 0>}
  n126 = Multiply(A=@n124, B=@n125)
  n127 = CrossProduct(A=@n123, B=@n125)
  n128 = Multiply(A=CrossProduct(A=@n123, B=Add(A=@n126, B=@n127) {constB=2})) {constB=2}
  n129 = TextureSample(UVs=@n47, Tex=@n120) {texture=&Black}
  n130 = If(A=@n10, B=Constant() {r=0.5}, "A > B"=@n129.RGBA, "A == B"=@n129.RGBA, "A < B"=ConstantBiasScale(None=@n129.RGBA) {bias=-0.5, scale=2}) {constB=0.5}
  n131 = ComponentMask(None=@n130) {b=true, desc=Quaternion XYZ, g=true, r=true}
  n132 = ComponentMask(None=@n130) {a=true, desc=Quaternion W}
  n133 = Constant3Vector() {constant=<rgba 0 0 1 0>}
  n134 = Add(A=Multiply(A=@n132, B=@n133), B=CrossProduct(A=@n131, B=@n133)) {constB=2}
  n135 = Add(A=Multiply(A=CrossProduct(A=@n131, B=@n134)) {constB=2}, B=@n133) {constB=2, desc=Normal}
  n136 = StaticSwitchParameter(True=LinearInterpolate(A=@n135, B=Add(A=@n128, B=@n125) {constB=2, desc=Normal}, Alpha=@n55), False=@n135) {
   desc=Interpolates interframe data when the animation frame is a fractional number. This results in smooth visuals even when you slow down the animation or when the frame rate is unstable.
   group=Houdini VAT
   parameterName=Interframe Interpolation
   sortPriority=1
  }
  n137 = Normalize(VectorInput=Transform(None=@n136) {transformSourceType=TRANSFORMSOURCE_Local})
  n138 = ComponentMask(None=@n137) {desc=Customized UV2, g=true, r=true}
  n139 = StaticSwitchParameter(True=@n119, False=@n138) {
   desc="Uses the medium-quality compressed normals from Position Texture Alpha. Only turn this on if Export Compressed Normals in Position Alpha is turned on in Houdini. This option removes the need for Rotation Texture, but the drawback is that you cannot apply any surface normal maps. If you do not need surface normal maps but still prefer higher-quality normals, turn off this and turn off Support Surface Normal Maps to retrieve normals from Rotation Texture."
   group=Houdini VAT
   parameterName=Use Compressed Normals
   sortPriority=4
  }
  n140 = StaticSwitchParameter(True=@n139, False=@n138) {
   desc="The data in Position Texture Alpha can come from the compressed normals or a custom float attribute. Turn this off if they come from the latter or if there is no data. It will disallow the use of compressed normals and interprets the data in this channel as a float, as opposed to a compressed 3D vector. This Material Function's Data in Position Alpha This Frame (VS) and Data in Position Alpha Next Frame (VS) outputs will always give you the appropriate interpretation of the data based on this parameter."
   group=Houdini VAT - Customization
   parameterName=Compressed Normals Are in Position Alpha
   sortPriority=0
  }
  FunctionOutput(None=@n140) {description=Connect this to Customized UV2., outputName=Customized UV2, sortPriority=22}
  n141 = Constant3Vector() {constant=<rgba 1 0 0 0>}
  n142 = Multiply(A=@n124, B=@n141)
  n143 = CrossProduct(A=@n123, B=@n141)
  n144 = Multiply(A=CrossProduct(A=@n123, B=Add(A=@n142, B=@n143) {constB=2})) {constB=2}
  n145 = Constant3Vector() {constant=<rgba 1 0 0 0>}
  n146 = Add(A=Multiply(A=@n132, B=@n145), B=CrossProduct(A=@n131, B=@n145)) {constB=2}
  n147 = Add(A=Multiply(A=CrossProduct(A=@n131, B=@n146)) {constB=2}, B=@n145) {constB=2, desc=Tangent}
  n148 = StaticSwitchParameter(True=LinearInterpolate(A=@n147, B=Add(A=@n144, B=@n141) {constB=2, desc=Tangent}, Alpha=@n55), False=@n147) {
   desc=Interpolates interframe data when the animation frame is a fractional number. This results in smooth visuals even when you slow down the animation or when the frame rate is unstable.
   group=Houdini VAT
   parameterName=Interframe Interpolation
   sortPriority=1
  }
  n149 = StaticSwitchParameter(True=Normalize(VectorInput=Transform(None=@n148) {transformSourceType=TRANSFORMSOURCE_Local}), False=Constant3Vector()) {
   defaultValue=true
   desc="Turn this on if you need to apply surface normal maps. When importing the FBX, set Normal Import Method to \"Import Normals and Tangents\"."
   group=Houdini VAT
   parameterName=Support Surface Normal Maps
   sortPriority=5
  }
  n150 = AppendVector(A=ComponentMask(None=@n137) {b=true}, B=ComponentMask(None=@n149) {r=true}) {desc=Customized UV3}
  n151 = StaticSwitchParameter(True=@n119, False=@n150) {
   desc="Uses the medium-quality compressed normals from Position Texture Alpha. Only turn this on if Export Compressed Normals in Position Alpha is turned on in Houdini. This option removes the need for Rotation Texture, but the drawback is that you cannot apply any surface normal maps. If you do not need surface normal maps but still prefer higher-quality normals, turn off this and turn off Support Surface Normal Maps to retrieve normals from Rotation Texture."
   group=Houdini VAT
   parameterName=Use Compressed Normals
   sortPriority=4
  }
  n152 = StaticSwitchParameter(True=@n151, False=@n150) {
   desc="The data in Position Texture Alpha can come from the compressed normals or a custom float attribute. Turn this off if they come from the latter or if there is no data. It will disallow the use of compressed normals and interprets the data in this channel as a float, as opposed to a compressed 3D vector. This Material Function's Data in Position Alpha This Frame (VS) and Data in Position Alpha Next Frame (VS) outputs will always give you the appropriate interpretation of the data based on this parameter."
   group=Houdini VAT - Customization
   parameterName=Compressed Normals Are in Position Alpha
   sortPriority=0
  }
  FunctionOutput(None=@n152) {description=Connect this to Customized UV3., outputName=Customized UV3, sortPriority=23}
  n153 = ComponentMask(None=@n149) {b=true, desc=Customized UV4, g=true}
  n154 = StaticSwitchParameter(True=@n119, False=@n153) {
   desc="Uses the medium-quality compressed normals from Position Texture Alpha. Only turn this on if Export Compressed Normals in Position Alpha is turned on in Houdini. This option removes the need for Rotation Texture, but the drawback is that you cannot apply any surface normal maps. If you do not need surface normal maps but still prefer higher-quality normals, turn off this and turn off Support Surface Normal Maps to retrieve normals from Rotation Texture."
   group=Houdini VAT
   parameterName=Use Compressed Normals
   sortPriority=4
  }
  n155 = StaticSwitchParameter(True=@n154, False=@n153) {
   desc="The data in Position Texture Alpha can come from the compressed normals or a custom float attribute. Turn this off if they come from the latter or if there is no data. It will disallow the use of compressed normals and interprets the data in this channel as a float, as opposed to a compressed 3D vector. This Material Function's Data in Position Alpha This Frame (VS) and Data in Position Alpha Next Frame (VS) outputs will always give you the appropriate interpretation of the data based on this parameter."
   group=Houdini VAT - Customization
   parameterName=Compressed Normals Are in Position Alpha
   sortPriority=0
  }
  FunctionOutput(None=@n155) {description=Connect this to Customized UV4., outputName=Customized UV4, sortPriority=24}
  n156 = StaticSwitchParameter(
   True=Normalize(VectorInput=Transform(None=@n97) {transformSourceType=TRANSFORMSOURCE_Local})
   False=Multiply(A=@n68.A, B=@n13) {desc=Original custom data this frame}
  ) {
   desc="The data in Position Texture Alpha can come from the compressed normals or a custom float attribute. Turn this off if they come from the latter or if there is no data. It will disallow the use of compressed normals and interprets the data in this channel as a float, as opposed to a compressed 3D vector. This Material Function's Data in Position Alpha This Frame (VS) and Data in Position Alpha Next Frame (VS) outputs will always give you the appropriate interpretation of the data based on this parameter."
   group=Houdini VAT - Customization
   parameterName=Compressed Normals Are in Position Alpha
   sortPriority=0
  }
  FunctionOutput(None=@n156) {
   description="Unlike other texture channels, Position Texture Alpha tends to store data in a special way. The data usually go through some form of normalization or compression. This output gives you this frame's data in their originally intended format, not the raw data in Position Texture Alpha. To get interpolated data, use this with Data in Position Alpha Next Frame (VS) and Interframe Interpolation Alpha (VS) to create your own Lerp."
   outputName="Data in Position Alpha This Frame (VS)"
   sortPriority=9
  }
  FunctionOutput(None=@n55) {
   description="Outputs the fractional part of the frame number on the vertex shader (VS). This can be used to interpolate various outputs. All outputs ending with (PS) can be automatically interpolated based on the Material Instance settings. All outputs ending with This Frame (VS) / Next Frame (VS) can be interpolated explicity downstream from this Material Function. This flexibility allows you to decide which sets of data need interpolation and how to implement the interpolation method. It can be linear, spherical linear, or more."
   outputName="Interframe Interpolation Alpha (0-1) (VS)"
   sortPriority=15
  }
  n157 = StaticSwitchParameter(True=Constant() {r=1}, False=ComponentMask(None=@n58.PS) {a=true}) {
   desc="To use this feature, Include Hidden Debug Plane in Geometry must be turned on in Houdini and this Material Function's Color RGB (PS) output must be connected to Base Color (at least during debugging). In the debug mode, the data embedded in the static mesh are displayed on a plane (facing local-space +Y).\r\n\r\nThe greyscale XYZ values are: \r\n\r\ntexture format and the number of position textures (0.7 or 0.8 = HDR; 0.3 or 0.4 = non-HDR; 0.3 or 0.7 = one texture; 0.4 or 0.8 = two textures);\r\n\r\nmax value of a custom attribute in Position Texture Alpha (no data if no custom attribute is present; compressed normals are not considered a custom attribute);\r\n\r\nno data.\r\n\r\nThe RGB values are: \r\n\r\npowers of two mode active pixels ratio X;\r\n\r\npowers of two mode active pixels ratio Y;\r\n\r\nanimation frame count.\r\n\r\n(Due to some numerical errors in the Unreal debug nodes, occasionally when an embedded value is actually 1 or 0, the plane may misleadingly display 0 or 1 . The does not affect the actual animation and you can safely ignore it. You can verify this is indeed a display bug by temporarily add or remove a few zeros in Embedded Data Rounding.)\r\n\r\nWhen not in the debug mode, the debug plane is invisible and incurs next to no cost."
   group=Houdini VAT - Debug
   parameterName=Debug Embedded Data
   sortPriority=3
  }
  FunctionOutput(None=@n157) {
   description="Outputs Color Texture Alpha on the pixel shader (PS). The data can come from the \"Alpha\" attribute or a custom float attribute. The data can be automatically interpolated between frames based on the Material Instance settings. To process this frame's data on the vertex shader, use Color RGBA This Frame (VS). To get interpolated data on the vertex shader, use Color RGBA This Frame (VS), Color RGBA Next Frame (VS), and Interframe Interpolation Alpha (VS) to create your own Lerp."
   outputName="Color Alpha (PS)"
   sortPriority=1
  }
 ]
}
