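# Translucent, two-sided unlit material for the KelpLinen HandGlow effect, a soft gradient glow with streaky noise.
# Emissive blends two colors along U with an intensity boost and time-based sine beat pulse.
# Opacity combines side fade, noise-varied top fade, panning cloud noise, location fade and depth fade, which Low and Medium quality skip.
# Attributes pass through an audio-reactive function; a soft-body VAT function is present but unconnected.
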Material /Clyde/Environments/KelpLinen/Effects/HandGlow/M_KelpLinen_HandGlow
graph {
 properties {
  bAutomaticallySetUsageInEditor=false
  bUseMaterialAttributes=true
  blendMode=BLEND_Translucent
  shadingModel=MSM_Unlit
  twoSided=true
  usedShadingModels=Unlit
 }
 nodes [
  n1 = TextureCoordinate()
  n2 = Power(Base=ComponentMask(None=OneMinus(None=@n1)) {r=true}, Exp=ScalarParameter() {defaultValue=2.2, parameterName=IntensityBoost_Power})
  n3 = ScalarParameter() {defaultValue=1, parameterName=IntensityBoost_Mult}
  n4 = LinearInterpolate(
   A=ScalarParameter() {defaultValue=1, parameterName=IntensityBoost_A}
   B=ScalarParameter() {defaultValue=1, parameterName=IntensityBoost_B}
   Alpha=Clamp(None=Multiply(A=@n2, B=@n3))
  )
  n5 = Power(Base=ComponentMask(None=TextureCoordinate()) {r=true}, Exp=ScalarParameter() {defaultValue=1, parameterName=Color_Grad_Power}) {constExponent=2.2}
  n6 = ScalarParameter() {defaultValue=1, parameterName=Color_Grad_Mult}
  n7 = Multiply(
   A=LinearInterpolate(
    A=VectorParameter() {parameterName=Color_A, defaultValue=<rgba 0 1.4645 2 1>}.RGB
    B=VectorParameter() {parameterName=Color_B, defaultValue=<rgba 6 2.0156 3.0924 1>}.RGB
    Alpha=Multiply(A=@n5, B=@n6)
   )
   B=LinearInterpolate(A=ScalarParameter() {defaultValue=0.85, group=Beat, parameterName=Glow_Beat_Value}, Alpha=NamedRerouteUsage()) {constA=1}
  )
  n8 = TextureCoordinate()
  n9 = Power(Base=Sine(None=ComponentMask(None=@n8) {g=true}) {period=2}, Exp=ScalarParameter() {defaultValue=3, parameterName=SideFade_Power})
  n10 = ScalarParameter() {defaultValue=1, parameterName=SideFade_Mult}
  n11 = TextureCoordinate()
  n12 = Panner(Coordinate=Divide(A=WorldPosition().XY) {constB=100000}, Speed=AppendVector(A=Constant() {r=0.1}, B=Constant()))
  n13 = Power(
   Base=OneMinus(None=ComponentMask(None=@n11) {r=true})
   Exp=MaterialFunctionCall(
    Input=TextureSample(UVs=@n12) {texture=/Game/Characters/Player/Male/Medium/Bodies/M_MED_Viper/Textures/Glitch/T_SoftNoise}.R
    Input Low=Constant()
    Input High=Constant() {r=1}
    Target Low=ScalarParameter() {defaultValue=4, parameterName=TopFade_Power_Min}
    Target High=ScalarParameter() {defaultValue=8, parameterName=TopFade_Power_Max}
   ) {materialFunction=&RemapValueRange}.Result
  )
  n14 = ScalarParameter() {defaultValue=3, parameterName=TopFade_Mult}
  n15 = Constant()
  n16 = Constant() {r=1}
  n17 = Panner(
   Coordinate=Multiply(A=TextureCoordinate(), B=AppendVector(A=@n15, B=@n16))
   Speed=AppendVector(A=ScalarParameter() {defaultValue=0.1, parameterName=Noise_Pan_X}, B=Constant())
  )
  n18 = Multiply(
   A=Multiply(A=Multiply(A=@n9, B=@n10), B=Multiply(A=@n13, B=@n14))
   B=Power(
    Base=TextureSample(UVs=@n17) {texture=/Clyde/ExtraDependencies/AtmosphericCloudNoise01}.R
    Exp=ScalarParameter() {defaultValue=1, parameterName=Noise_Power}
   )
  )
  n19 = Multiply(A=@n18, B=DepthFade(FadeDistance=ScalarParameter() {defaultValue=5000, parameterName=DepthFade_Distance}) {fadeDistanceDefault=5000})
  n20 = QualitySwitch(Default=@n19, Low=@n18, High=@n19, Medium=@n18, Epic=@n19)
  n21 = MaterialFunctionCall() {materialFunction=/ClydeCosmetics/Environments/KelpLinen/MF_KelpLinen_LocationFade}
  MaterialFunctionCall!(In=SetMaterialAttributes!(MaterialAttributes, Emissive Color=Multiply(A=@n4, B=@n7), Opacity=Multiply(A=@n20, B=@n21.Output)) {attributeSetTypes [#1, #2]}) {materialFunction=/ClydeCosmetics/Environments/KelpLinen/MF_Kelp_AudioReactive}
  Multiply(A=TextureCoordinate(), B=AppendVector(A=Constant(), B=Constant() {r=1}))
  n22 = Time()
  n23 = Multiply(A=ScalarParameter() {defaultValue=0.1, desc=Sudo how long the pulse takes, group=Beat, parameterName=Time_Multi}) {constB=4, desc=Number of waves in water}
  NamedRerouteDeclaration(None=Sine(None=Multiply(A=@n22, B=@n23) {constB=0.2})) {name=Music_Beat, nodeColor=<rgba 0.125 1 0 1>}
  Reroute(None=MaterialFunctionCall!("Time (VS)"=Time(), "Display Frame Override (VS)", "Surface Normal Maps RGB (PS)", "Additional Local Position Offset (VS)") {materialFunction=/Clyde/Environments/KelpLinen/Effects/HandGlow/VAT/MF_VAT_SoftBodyDeformation_KelpLinen}.World Position Offset) {desc=WPO}
  ScalarParameter() {defaultValue=6, parameterName=TopFade_Power}
 ]
}
