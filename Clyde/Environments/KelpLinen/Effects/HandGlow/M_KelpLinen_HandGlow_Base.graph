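# Unlit base material for the KelpLinen hand glow effect, built for Niagara mesh particles.
# Panning fire noise distorts a second sample of itself, driving a two-color blend and an opacity term.
# UV V gradients boost one end's intensity, while an inverted Fresnel term adds edge color.
# A time-panned, world-position noise lookup, scaled by the Noise Scale dynamic parameter, computes displacement along vertex normals.
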Material /Clyde/Environments/KelpLinen/Effects/HandGlow/M_KelpLinen_HandGlow_Base
graph {
 properties {
  bAutomaticallySetUsageInEditor=false
  bUsedWithNiagaraMeshParticles=true
  previewMesh=&/SM_KelpLinen_sphere_HP
  shadingModel=MSM_Unlit
  usedShadingModels=Unlit
 }
 nodes [
  n1 = Multiply(A=TextureCoordinate(), B=AppendVector(A=Constant() {r=1}, B=Constant() {r=1}))
  n2 = AppendVector(A=ScalarParameter() {defaultValue=0.05, parameterName=Noise_Pan_X}, B=ScalarParameter() {defaultValue=0.025, parameterName=Noise_Pan_Y})
  n3 = Multiply(A=TextureCoordinate(), B=AppendVector(A=Constant() {r=1}, B=Constant() {r=1}))
  n4 = AppendVector(A=ScalarParameter() {defaultValue=0.06, parameterName=D_Pan_X}, B=ScalarParameter() {defaultValue=0.026, parameterName=D_Pan_Y})
  n5 = Power(
   Base=TextureSample(UVs=Panner(Coordinate=@n3, Speed=@n4)) {texture=/Game/Global/Textures/Default/Noise/T_FireNoise_02}.R
   Exp=ScalarParameter() {defaultValue=4, parameterName=D_Exp}
  )
  n6 = ScalarParameter() {defaultValue=1, parameterName=D_Power}
  n7 = TextureSample(UVs=Add(A=Panner(Coordinate=@n1, Speed=@n2), B=Multiply(A=@n5, B=@n6))) {texture=&T_FireNoise_02}
  n8 = Power(Base=@n7.RGB, Exp=ScalarParameter() {parameterName=Noise_Opacity_Power})
  n9 = ScalarParameter() {defaultValue=10, parameterName=Noise_Opacity_Mult}
  n10 = Clamp(None=Clamp(None=Multiply(A=@n8, B=@n9)))
  n11 = ComponentMask(None=TextureCoordinate()) {g=true}
  n12 = Clamp(None=Multiply(A=Power(Base=@n11) {constExponent=4}, B=ScalarParameter() {defaultValue=1, parameterName=Bottom_Opacity_Boost}))
  n13 = Multiply(A=@n12, B=ScalarParameter() {defaultValue=2, parameterName=LowerHalf_Color_Intensity}) {constB=10}
  Clamp(None=Add(A=Multiply(A=@n10, B=@n13), B=@n10))
  Constant()
  Constant() {r=0.5}
  Constant() {r=0.2}
  n14 = Time()
  n15 = Add(
   A=Divide(A=WorldPosition().XYZ, B=DynamicParameter() {paramNames [Emissive, Noise Scale, Param3, Param4]}.Noise Scale)
   B=Multiply(A=Multiply(A=@n14) {constB=0.1}, B=VectorParameter() {parameterName=WPO_Noise_Pan_Speed, defaultValue=<rgba 0 0 1 1>}.RGB)
  )
  Multiply(
   A=LinearInterpolate(
    A=ScalarParameter() {parameterName=WPO_Intensity_Min}
    B=ScalarParameter() {defaultValue=8000, parameterName=WPO_Intensity_Max}
    Alpha=TextureSample(UVs=@n15) {samplerType=SAMPLERTYPE_LinearGrayscale, texture=&/VT_ClothSimWPO_KelpLinen}.RGB
   )
   B=VertexNormalWS()
  )
  n16 = Power(Base=ComponentMask(None=TextureCoordinate()) {g=true}, Exp=ScalarParameter() {defaultValue=6, parameterName=Bottom_EmissiveBoost_Power}) {constExponent=4}
  n17 = ScalarParameter() {defaultValue=10, parameterName=Bottom_Emissive_Boost}
  n18 = MaterialFunctionCall(
   Input=Clamp(None=Multiply(A=@n16, B=@n17))
   Target Low=Constant() {r=1}
   Target High=ScalarParameter() {defaultValue=1, parameterName=Bottom_Emissive_Boost_Intensity}
  ) {materialFunction=&RemapValueRange}
  n19 = VectorParameter() {parameterName=Noise_Color_A, defaultValue=<rgba 3.634 1.0797 10 1>}
  n20 = Power(Base=@n7.RGB, Exp=ScalarParameter() {defaultValue=2, parameterName=Noise_Color_Power})
  n21 = ScalarParameter() {defaultValue=3, parameterName=Noise_Color_Mult}
  n22 = LinearInterpolate(
   A=@n19.RGB
   B=VectorParameter() {parameterName=Noise_Color_B, defaultValue=<rgba 0.3049 1 0.9189 1>}.RGB
   Alpha=Clamp(None=Multiply(A=@n20, B=@n21))
  )
  n23 = ScalarParameter() {defaultValue=2, parameterName=Edge_Color_Mult}
  n24 = MaterialFunctionCall(
   Brightness=ScalarParameter() {defaultValue=1, parameterName=Edge_Brightness}
   Exponent=ScalarParameter() {defaultValue=1, parameterName=Edge_Exp}
  ) {materialFunction=&MF_Fresnel_FX}
  n25 = ScalarParameter() {defaultValue=8, parameterName=Edge_Power}
  n26 = Clamp(None=Multiply(A=Power(Base=@n24.Fresnel Inverted, Exp=@n25), B=ScalarParameter() {defaultValue=2, parameterName=Edge_Mult}))
  n27 = Constant()
  n28 = ScalarParameter() {defaultValue=2, parameterName=Edge_Color_Intensity}
  Multiply(
   A=MaterialFunctionCall(Input=@n12, Target Low=ScalarParameter() {defaultValue=0.2}, Target High=ScalarParameter() {defaultValue=3, parameterName=Param_1}) {materialFunction=&RemapValueRange}.Result
   B=LinearInterpolate(
    A=Multiply(A=@n18.Result, B=@n22)
    B=Multiply(A=@n19.RGB, B=@n23)
    Alpha=MaterialFunctionCall(Input=@n26, Target Low=@n27, Target High=@n28) {materialFunction=&RemapValueRange}.Result
   )
  )
 ]
}
