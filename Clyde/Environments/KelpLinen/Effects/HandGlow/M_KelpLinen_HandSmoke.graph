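# Unlit, two-sided translucent smoke material for the KelpLinen hand glow effect.
# Emissive blends two colors and intensity boosts along U, pulsed by a time-based sine named Music_Beat.
# Opacity uses panning smoke distorted by a normal map, with U fade, quality-switched depth fade and location fade.
# Attributes pass through an audio-reactive function, then get U-weighted velocity WPO; sway, noise and VAT branches appear unconnected.
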Material /Clyde/Environments/KelpLinen/Effects/HandGlow/M_KelpLinen_HandSmoke
graph {
 properties {
  bAutomaticallySetUsageInEditor=false
  bUseMaterialAttributes=true
  blendMode=BLEND_Translucent
  shadingModel=MSM_Unlit
  twoSided=true
  usedShadingModels=Unlit
 }
 nodes [
  n1 = ScalarParameter() {defaultValue=1, parameterName=IntensityBoost_A}
  n2 = ScalarParameter() {defaultValue=1, parameterName=IntensityBoost_B}
  n3 = ComponentMask(None=OneMinus(None=TextureCoordinate())) {r=true}
  n4 = ScalarParameter() {defaultValue=2.2, parameterName=IntensityBoost_Power}
  n5 = Clamp(None=Multiply(A=Power(Base=@n3, Exp=@n4), B=ScalarParameter() {defaultValue=1, parameterName=IntensityBoost_Mult}))
  n6 = ComponentMask(None=TextureCoordinate()) {r=true}
  n7 = ScalarParameter() {defaultValue=2.2, parameterName=Color_Grad_Power}
  n8 = LinearInterpolate(
   A=VectorParameter() {parameterName=Color_A, defaultValue=<rgba 0 6.5057 10 1>}.RGB
   B=VectorParameter() {parameterName=Color_B, defaultValue=<rgba 0 6 2.6521 1>}.RGB
   Alpha=Multiply(A=Power(Base=@n6, Exp=@n7) {constExponent=2.2}, B=ScalarParameter() {defaultValue=3, parameterName=Color_Grad_Mult})
  )
  n9 = LinearInterpolate(A=ScalarParameter() {defaultValue=0.85, group=Beat, parameterName=Glow_Beat_Value}, Alpha=NamedRerouteUsage())
  n10 = ComponentMask(None=TextureCoordinate()) {r=true}
  n11 = Multiply(
   A=Power(Base=OneMinus(None=@n10), Exp=ScalarParameter() {defaultValue=2.2, parameterName=TopFade_Power})
   B=ScalarParameter() {defaultValue=1, parameterName=TopFade_Mult}
  )
  n12 = ScalarParameter() {defaultValue=2, parameterName=Mask_Tile_X}
  n13 = ScalarParameter() {defaultValue=1, parameterName=Mask_Tile_Y}
  n14 = Panner(
   Coordinate=Multiply(A=TextureCoordinate(), B=AppendVector(A=@n12, B=@n13))
   Speed=AppendVector(A=ScalarParameter() {defaultValue=-0.05, parameterName=Mask_Pan_X}, B=Constant())
  )
  n15 = Multiply(A=TextureCoordinate(), B=AppendVector(A=ScalarParameter() {defaultValue=0.5, parameterName=MacroDistortion_Tile_X}, B=Constant() {r=1}))
  n16 = AppendVector(A=ScalarParameter() {defaultValue=-0.1, parameterName=MacroDistortion_Pan_X}, B=Constant())
  n17 = ComponentMask(None=TextureSample(UVs=Panner(Coordinate=@n15, Speed=@n16)) {samplerType=SAMPLERTYPE_Normal, texture=&T_Weap_Distort_N_02}.RGB) {g=true, r=true}
  n18 = ScalarParameter() {defaultValue=1, parameterName=Macro_Distortion_Power}
  n19 = Multiply(
   A=Multiply(A=Power(Base=@n17, Exp=@n18), B=ScalarParameter() {defaultValue=0.2, parameterName=Distortion_Intensity})
   B=AppendVector(A=Constant() {r=1}, B=Constant() {r=1})
  )
  n20 = Constant()
  n21 = ScalarParameter() {parameterName=Distortion_Offset_X}
  n22 = TextureSample(UVs=Add(A=Add(A=@n14, B=@n19) {constB=0}, B=AppendVector(A=@n20, B=@n21))) {
   samplerSource=SSM_Wrap_WorldGroupSettings
   samplerType=SAMPLERTYPE_Grayscale
   texture=/Game/Effects/Fort_Effects/Textures/Smoke/T_Smoke_Trail_Soft_NoClamp
  }
  n23 = Multiply(
   A=Power(Base=@n22.RGB, Exp=ScalarParameter() {defaultValue=4, parameterName=Alpha_Power})
   B=ScalarParameter() {defaultValue=2, parameterName=Alpha_mult}
  ) {constB=2}
  n24 = Multiply(A=Clamp(None=Multiply(A=@n11, B=@n23)), B=ScalarParameter() {defaultValue=0.5, parameterName=Opacity_Master})
  n25 = Multiply(A=@n24, B=DepthFade(FadeDistance=ScalarParameter() {defaultValue=5000, parameterName=DepthFade_Distance}) {fadeDistanceDefault=5000})
  n26 = SetMaterialAttributes!(
   MaterialAttributes
   Emissive Color=Multiply(A=LinearInterpolate(A=@n1, B=@n2, Alpha=@n5), B=Multiply(A=@n8, B=@n9))
   Opacity=Multiply(
    A=QualitySwitch(Default=@n25, Low=@n24, High=@n25, Medium=@n24, Epic=@n25)
    B=MaterialFunctionCall() {materialFunction=/ClydeCosmetics/Environments/KelpLinen/MF_KelpLinen_LocationFade}.Output
   )
  ) {attributeSetTypes [#1, #2]}
  n27 = TextureCoordinate()
  n28 = Multiply(
   A=Power(Base=ComponentMask(None=@n27) {r=true}, Exp=ScalarParameter() {defaultValue=3, parameterName=WPO_Grad_Power})
   B=ScalarParameter() {defaultValue=3, parameterName=WPO_Grad_Mult}
  )
  BlendMaterialAttributes(A=@n26, Alpha=LinearInterpolate(A=Constant(), B=Clamp(None=@n28), Alpha=ScalarParameter() {defaultValue=1, parameterName=Use_WPO_Grad}))
  n29 = Time()
  n30 = ScalarParameter() {defaultValue=0.2, parameterName=WPOSwaySpeed.X}
  n31 = Multiply(A=@n29, B=@n30)
  n32 = ScalarParameter() {defaultValue=1, parameterName=WPOSwayAmouont.X}
  n33 = Multiply(A=@n29, B=@n30) {constB=0.2}
  n34 = Add(
   A=Multiply(A=Sine(None=@n31) {period=1.77}, B=Multiply(A=@n32) {constB=0.3})
   B=Multiply(A=Sine(None=@n33), B=ScalarParameter() {defaultValue=1, parameterName=WPOSwayAmouont.X})
  )
  Clamp(None=@n34) {minDefault=-1}
  n35 = Transform(None=Constant3Vector() {constant=<rgba 1 0 0 0>}) {transformSourceType=TRANSFORMSOURCE_Camera}
  n36 = Time()
  n37 = ScalarParameter() {defaultValue=0.1, parameterName=WPOSwaySpeed.Y}
  n38 = Multiply(A=Sine(None=Multiply(A=@n36, B=@n37) {constB=0.2}) {period=1.33}, B=ScalarParameter() {parameterName=WPOSwayAmouont.Y})
  n39 = Transform(None=Constant3Vector() {constant=<rgba 0 1 0 0>}) {transformSourceType=TRANSFORMSOURCE_Camera}
  n40 = TextureCoordinate()
  LinearInterpolate(B=Add(A=Multiply(A=@n34, B=@n35), B=Multiply(A=@n38, B=@n39)), Alpha=ComponentMask(None=Power(Base=@n40) {constExponent=3}) {r=true})
  n41 = VectorParameter() {parameterName=WPOVelocity, defaultValue=<rgba 0 0 0 1>}
  n42 = Constant() {r=-1000}
  n43 = Constant() {r=1000}
  n44 = Multiply(
   A=MaterialFunctionCall(X=Clamp(None=@n41.R, Min=@n42, Max=@n43), Y=Clamp(None=@n41.G, Min=@n42, Max=@n43), Z=Clamp(None=@n41.B, Min=@n42, Max=@n43)) {materialFunction=&MakeFloat3}.Result
   B=@n41.A
  )
  Multiply(A=@n44, B=Transform(None=Constant3Vector() {constant=<rgba 1 0 0 0>}) {transformSourceType=TRANSFORMSOURCE_Camera})
  n45 = Divide(A=ComponentMask(None=WorldPosition().XYZ) {b=true, g=true}, B=ScalarParameter() {defaultValue=100000, parameterName=WPO_Noise_TileSize})
  n46 = AppendVector(A=ScalarParameter() {defaultValue=-0.05, parameterName=WPO_Noise_Pan_X}, B=Constant() {r=0.001})
  Multiply(
   A=TextureSample(UVs=Panner(Coordinate=@n45, Speed=@n46)) {texture=/Game/Characters/Player/Male/Medium/Bodies/M_MED_Viper/Textures/Glitch/T_SoftNoise}.RGB
   B=ScalarParameter() {defaultValue=20000, parameterName=WPO_Noise_Amount}
  )
  n47 = Time()
  n48 = Multiply(A=ScalarParameter() {defaultValue=0.1, desc=Sudo how long the pulse takes, group=Beat, parameterName=Time_Multi}) {constB=4, desc=Number of waves in water}
  NamedRerouteDeclaration(None=Sine(None=Multiply(A=@n47, B=@n48) {constB=0.2})) {name=Music_Beat, nodeColor=<rgba 0.125 1 0 1>}
  Reroute(None=MaterialFunctionCall!("Time (VS)"=Time(), "Display Frame Override (VS)", "Surface Normal Maps RGB (PS)", "Additional Local Position Offset (VS)") {materialFunction=/Clyde/Environments/KelpLinen/Effects/HandGlow/VAT/MF_VAT_SoftBodyDeformation_KelpLinen}.World Position Offset)
  n49 = Power(Base=TextureCoordinate()) {constExponent=3}
  SetMaterialAttributes!(
   MaterialAttributes=MaterialFunctionCall!(In=@n26) {materialFunction=/ClydeCosmetics/Environments/KelpLinen/MF_Kelp_AudioReactive}.Result
   World Position Offset=LinearInterpolate(B=@n44, Alpha=ComponentMask(None=@n49) {r=true})
  ) {attributeSetTypes [#3]}
  TextureCoordinate()
 ]
}
